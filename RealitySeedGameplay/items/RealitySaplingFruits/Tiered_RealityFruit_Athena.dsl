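; Reality Sapling fruit, a pickable loot container that ripens from baby through youth to a rarity tier, with separate mythic stages.
; Server timers advance growth toward a tier set by the parent sapling's weed count, with non-mythic ripening capped at gold.
; The replicated variant drives color and scale lerps, interaction, prompt text and loot tier group.
; Fallen fruit resumes after pawn or projectile hits, opens on mounted turrets, decays, and corrects client landing location.
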(fn AutoLootFruit
  (params (NumTimesWeeded :type Integer))
  (if (not (Variables|Default|GetIsMythicFruit))
    (switch Utilities|FlowControl|Switch|SwitchonSaplingFruitVariants (Variables|Default|GetVariantOrRarity)
      (:NewEnumerator1
        _)
      (:NewEnumerator9
        _)
      (:NewEnumerator2
        ;[1 1 1] Ripen straight to the destination variant
        (Container|SpawnLoot 0 :!label _label))
      (:NewEnumerator3
        (goto _label))
      (:NewEnumerator4
        (goto _label))
      (:NewEnumerator5
        (goto _label))
      (:NewEnumerator6
        (goto _label))
      (:NewEnumerator7
        (goto _label))
      (:NewEnumerator8
        (goto _label)))
    (LocalVariables|SetGray%20As%20An%20Int (Math|Conversions|ToInteger(Byte) (Variables|Default|GetGray)))
    (LocalVariables|SetMythic%20baby%20As%20An%20Int (Math|Conversions|ToInteger(Byte) (Variables|Default|GetMythicBaby)))
    (bind _output_get (LocalVariables|SetVariant%20As%20An%20Enum (Utilities|Enum|BytetoEnumSaplingFruitVariants (Math|Conversions|ToByte(Integer) (Utilities|Operators|Add NumTimesWeeded (LocalVariables|GetGray%20As%20An%20Int) 0)))))
    (if (>= (Math|Conversions|ToInteger(Byte) _output_get) (LocalVariables|GetMythic%20baby%20As%20An%20Int))
      ;;[1 1 1] Limit to gold.
      (LocalVariables|SetVariant%20As%20An%20Enum "NewEnumerator6"))
    (Variables|Default|SetVariantOrRarity (LocalVariables|GetVariant%20As%20An%20Enum))
    (goto _label)))

(fn DestroyThisFruit
  (Networking|SwitchHasAuthority
    (:Authority
      (Variables|Container|SetbForceSpawnLootOnDestruction false)
      (Gameplay|SilentDie false))))

(fn ForceOpenFruit
  (Networking|SwitchHasAuthority
    (:Authority
      (if (not (Variables|Default|GetOpened))
        (Variables|Default|SetOpened true)
        (Networking|ForceNetUpdate)
        (Container|SpawnLoot 0)
        (Actor|SetLifeSpan 7.0)))))

(fn ForceSetLocationOnLanding)

(fn GetColorByRarity
  (params (InColor :type SaplingFruitVariants%20Enum :default "NewEnumerator0") (IsInner :type Boolean))
  (switch Utilities|FlowControl|Switch|SwitchonSaplingFruitVariants InColor
    (:NewEnumerator0
      _)
    (:NewEnumerator1
      _)
    (:NewEnumerator9
      ;[1 1 1] Youth
      (if IsInner
        (return (Variables|ColorConsts|GetDestColor_Youth_Inner))
        (else
          (return (Variables|ColorConsts|GetDestColor_Youth_Outer)))))
    (:NewEnumerator2
      ;[1 1 1] Grey
      (if IsInner
        (return (Variables|ColorConsts|GetDest%20Color_Common_Inner))
        (else
          (return (Variables|ColorConsts|GetDest%20Color_Common_Outer)))))
    (:NewEnumerator3
      ;[1 1 1] Green
      (if IsInner
        (return (Variables|ColorConsts|GetDest%20Color_Uncommon_Inner))
        (else
          (return (Variables|ColorConsts|GetDest%20Color_Uncommon_Outer)))))
    (:NewEnumerator4
      ;[1 1 1] Blue
      (if IsInner
        (return (Variables|ColorConsts|GetDest%20Color_Rare_Inner))
        (else
          (return (Variables|ColorConsts|GetDest%20Color_Rare_Outer)))))
    (:NewEnumerator5
      ;[1 1 1] Purple
      (if IsInner
        (return (Variables|ColorConsts|GetDest%20Color_Epic_Inner))
        (else
          (return (Variables|ColorConsts|GetDest%20Color_Epic_Outer)))))
    (:NewEnumerator6
      ;[1 1 1] Gold
      (if IsInner
        (return (Variables|ColorConsts|GetDest%20Color_Legendary_Inner))
        (else
          (return (Variables|ColorConsts|GetDest%20Color_Legendary_Outer)))))
    (:NewEnumerator7
      ;[1 1 1] Mythic Baby
      (if IsInner
        (return (Variables|ColorConsts|GetDest%20Color_Mythic_Baby_Inner))
        (else
          (return (Variables|ColorandScale|GetDest%20Color_Mythic_Baby_Outer)))))
    (:NewEnumerator8
      ;[1 1 1] Mythic
      (if IsInner
        (return (Variables|ColorConsts|GetDest%20Color_Mythic_Inner))
        (else
          (return (Variables|ColorConsts|GetDest%20Color_Mythic_Outer))))))
  ;[1 1 1] Picked or Baby
  (if IsInner
    (return (Variables|ColorConsts|GetDestColor_Baby_Inner))
    (else
      (return (Variables|ColorConsts|GetDestColor_Baby_Outer)))))

(fn GetSaplingVariantorRarity
  (params (NumTimesWeeded :type Integer))
  (LocalVariables|SetGray%20As%20An%20Int (Math|Conversions|ToInteger(Byte) (Variables|Default|GetGray)))
  (LocalVariables|SetMythic%20baby%20As%20An%20Int (Math|Conversions|ToInteger(Byte) (Variables|Default|GetMythicBaby)))
  ;[1 1 1] Ripen straight to the destination variant
  (LocalVariables|SetVariant%20As%20An%20Enum (Utilities|Enum|BytetoEnumSaplingFruitVariants (Math|Conversions|ToByte(Integer) (Utilities|Operators|Add NumTimesWeeded (LocalVariables|GetGray%20As%20An%20Int) 0))))
  (return (LocalVariables|GetVariant%20As%20An%20Enum)))

(fn GetWeightedFruitColor
  (params (SourceColor :type Linear%20Color%20Structure) (AlphaFactor :type Float%20%28double-precision%29) (IsInner :type Boolean))
  (bind (_fruitreppeddata_owningplayercontroller_6_7553902549a37048686212bc13e7d8ca _fruitreppeddata_saplingparent_11_8f3a618f4bd64834d3636281b025ea9d) (Utilities|Struct|BreakFruitReppedData (Variables|GetFruitReppedData_OwningPlayerController_6_7553902549A37048686212BC13E7D8CA)))
  (Utilities|FlowControl|Sequence
    (:then_0
      (Utilities|IsValid _fruitreppeddata_saplingparent_11_8f3a618f4bd64834d3636281b025ea9d
        (:Is%20Valid
          (bind _outrarityorvariant (CallFunction|GetSaplingVariantOrRarity (Variables|/RealitySeedGameplay/items/RealitySaplings/B_SMA_Athena_RealitySapling_Parent.B_SMA_Athena_RealitySapling_Parent_C|GetNumTimesWeeded _fruitreppeddata_saplingparent_11_8f3a618f4bd64834d3636281b025ea9d)))
          (bind _outcolor (CallFunction|GetColorbyRarity _outrarityorvariant IsInner))
          (LocalVariables|SetDestColor _outcolor))
        (:Is%20Not%20Valid
          (bind _outcolor_2 (CallFunction|GetColorbyRarity (Variables|Default|GetVariantOrRarity) IsInner))
          (LocalVariables|SetDestColor _outcolor_2))))
    (:then_1
      (return (Math|Color|Lerp(LinearColor) SourceColor (LocalVariables|GetDestColor) AlphaFactor)))))

(fn GetFruitRarityRichText
  (bind _result (ArrayMacros|SafeArrayGet (Variables|Default|GetRarityTexts) (Utilities|Select :Index (Variables|Default|GetVariantOrRarity) :NewEnumerator0 0 :NewEnumerator1 0 :NewEnumerator9 0 :NewEnumerator2 0 :NewEnumerator3 1 :NewEnumerator4 2 :NewEnumerator5 3 :NewEnumerator6 4 :NewEnumerator7 5 :NewEnumerator8 5))
    (:%20%20
      (return _result))))

(fn GetScalableFloatTime
  (params (ScalableFloat :type Scalable%20Float%20Structure))
  (if (Variables|Debug|GetUseShortTimes)
    (return (+ (ScalableFloat|GetValueatLevel (Variables|Default|GetShortTime)) (* (Math|Conversions|ToFloat(Integer) (Variables|Default|GetFruitSlot)) 1.5)))
    (else
      (return (+ (ScalableFloat|GetValueatLevel ScalableFloat) (* (Math|Conversions|ToFloat(Integer) (Variables|Default|GetFruitSlot)) 1.5))))))

(fn HandleProjectileMovementStop
  (params (ImpactResult :type Hit%20Result%20Structure))
  (bind (_bblockinghit _binitialoverlap _time _distance _location _impactpoint _normal _impactnormal _physmat _hitactor) (Collision|BreakHitResult ImpactResult))
  (Utilities|Casting|CastToFortProjectileBase _hitactor
    (:then
      _)
    (:CastFailed
      (Utilities|Casting|CastToFortRemoteControlledPawnAthena _hitactor
        (:then
          _)
        (:CastFailed
          (Utilities|Casting|CastToPawn _hitactor
            (:then
              (Utilities|Casting|CastToFortMountedTurret _hitactor
                (:then
                  ;[1 1 1] If we land on a mounted Turret then go ahead and open
                  (Utilities|FlowControl|Sequence
                    (:then_0
                      (Utilities|Time|ClearTimerbyFunctionName 0 "Multicast_ApplyGravityForFall"))
                    (:then_1
                      (Networking|SwitchHasAuthority
                        (:Authority
                          (Variables|Default|SetLandingLocation (Transformation|GetActorLocation)))))
                    (:then_2
                      (Default|PlayLandingFX true))
                    (:then_3
                      (Default|ForceOpenFruit))))
                (:CastFailed
                  ;[1 1 1] If we stopped due to hitting a pawn / vehicle, after some delay attempt to resume falling
                  (Networking|SwitchHasAuthority
                    (:Authority
                      (Utilities|Time|SetTimerbyEvent (EventDispatchers|CreateEvent :!event Multicast_ApplyGravityForFall) 2.0))))))
            (:CastFailed
              ;[1 1 1] Handle landing
              (Utilities|FlowControl|Sequence
                (:then_0
                  (Utilities|Time|ClearTimerbyFunctionName 0 "Multicast_ApplyGravityForFall"))
                (:then_1
                  (Networking|SwitchHasAuthority
                    (:Authority
                      (Default|SetupBindtoBuildingActor _hitactor)
                      (Variables|Default|SetLandingLocation (Transformation|GetActorLocation)))))
                (:then_2
                  (Default|PlayLandingFX true)))))))))
  ;[1 1 1] If we just hit a projectile, resume falling
  (Networking|SwitchHasAuthority
    (:Authority
      (Utilities|Time|SetTimerbyEvent (EventDispatchers|CreateEvent :!event Multicast_ApplyGravityForFall) 1e-05))))

(fn IncrementGrowthState
  (bind (_fruitreppeddata_owningplayercontroller_6_7553902549a37048686212bc13e7d8ca _fruitreppeddata_saplingparent_11_8f3a618f4bd64834d3636281b025ea9d) (Utilities|Struct|BreakFruitReppedData (Variables|GetFruitReppedData_OwningPlayerController_6_7553902549A37048686212BC13E7D8CA)))
  (bind (_fruitreppeddata_owningplayercontroller_6_7553902549a37048686212bc13e7d8ca_2 _fruitreppeddata_saplingparent_11_8f3a618f4bd64834d3636281b025ea9d_2) (Utilities|Struct|BreakFruitReppedData (Variables|GetFruitReppedData_OwningPlayerController_6_7553902549A37048686212BC13E7D8CA)))
  (if (not (or (Utilities|Enum|Equal(Enum) (Variables|Default|GetVariantOrRarity) "NewEnumerator7") (Utilities|Enum|Equal(Enum) (Variables|Default|GetVariantOrRarity) "NewEnumerator8")))
    (Utilities|IsValid _fruitreppeddata_saplingparent_11_8f3a618f4bd64834d3636281b025ea9d_2
      (:Is%20Valid
        (LocalVariables|SetGrayAsInt (Math|Conversions|ToInteger(Byte) (Variables|Default|GetGray)))
        (LocalVariables|SetMythicBabyAsInt (Math|Conversions|ToInteger(Byte) (Variables|Default|GetMythicBaby)))
        (if (not (>= (Math|Conversions|ToInteger(Byte) (Variables|Default|GetVariantOrRarity)) (Utilities|Operators|Add (Variables|/RealitySeedGameplay/items/RealitySaplings/B_SMA_Athena_RealitySapling_Parent.B_SMA_Athena_RealitySapling_Parent_C|GetNumTimesWeeded _fruitreppeddata_saplingparent_11_8f3a618f4bd64834d3636281b025ea9d_2) (LocalVariables|GetGrayAsInt) 0)))
          (Networking|SwitchHasAuthority
            (:Authority
              (switch Utilities|FlowControl|Switch|SwitchonSaplingFruitVariants (Variables|Default|GetVariantOrRarity)
                (:NewEnumerator0
                  _)
                (:NewEnumerator1
                  _)
                (:NewEnumerator9
                  ;[1 1 1] Ripen straight to the destination variant
                  (bind _output_get (LocalVariables|SetVariantAsEnum (Utilities|Enum|BytetoEnumSaplingFruitVariants (Math|Conversions|ToByte(Integer) (Utilities|Operators|Add (Variables|/RealitySeedGameplay/items/RealitySaplings/B_SMA_Athena_RealitySapling_Parent.B_SMA_Athena_RealitySapling_Parent_C|GetNumTimesWeeded _fruitreppeddata_saplingparent_11_8f3a618f4bd64834d3636281b025ea9d) (LocalVariables|GetGrayAsInt) 0))) :!label _label))
                  (if (>= (Math|Conversions|ToInteger(Byte) _output_get) (LocalVariables|GetMythicBabyAsInt))
                    ;;[1 1 1] Limit to gold.
                    (LocalVariables|SetVariantAsEnum "NewEnumerator6"))
                  (Variables|Default|SetVariantOrRarity (LocalVariables|GetVariantAsEnum)))
                (:NewEnumerator2
                  (goto _label))
                (:NewEnumerator3
                  (goto _label))
                (:NewEnumerator4
                  (goto _label))
                (:NewEnumerator5
                  (goto _label)))
              ;[1 1 1] Increment to the next variant
              (bind _output_get_2 (LocalVariables|SetVariantAsInt (Math|Conversions|ToInteger(Byte) (Variables|Default|GetVariantOrRarity))))
              (bind _result (Math|Integer|IncrementInt _output_get_2)
                (:%20%20
                  (bind _output_get_3 (LocalVariables|SetVariantAsEnum (Utilities|Enum|BytetoEnumSaplingFruitVariants (Math|Conversions|ToByte(Integer) _result))))
                  (Variables|Default|SetVariantOrRarity _output_get_3))))))))))

(fn NotifySaplingUpgraded
  (if (not (Utilities|Time|IsValidTimerHandle (Variables|Default|GetGrowthTimer)))
    (bind _returnvalue (Utilities|Time|SetTimerforNextTickbyFunctionName 0 "Increment Growth State"))
    (Variables|Default|SetGrowthTimer _returnvalue)))

(fn OnRep_LandingLocation
  (Networking|SwitchHasAuthority
    (:Remote
      ;[1 1 1] Force client fruit to server location after a slight delay if we haven't stopped due to parts of level not being streamed in
      (if (Math|Vector|Equal(Vector) (Variables|Default|GetLandingLocation) "0, 0, 0" 0.1)
        (Utilities|Time|ClearTimerbyFunctionName 0 "ForceSetLocationOnLanding")
        (else
          (Utilities|Time|SetTimerbyFunctionName 0 "ForceSetLocationOnLanding" 0.5))))))

(fn OnRep_Opened)

(fn OnRep_UseCurvedStem
  (if true
    (Components|StaticMesh|SetStaticMesh (Variables|Default|GetStaticMeshStem) "/RealitySeedGameplay/Environment/Assets_RealityFlower/Meshes/Reality_Fruit_Stem_Curved.Reality_Fruit_Stem_Curved")))

(fn OnRep_VariantOrRarity
  (bind (_fruitreppeddata_owningplayercontroller_6_7553902549a37048686212bc13e7d8ca _fruitreppeddata_saplingparent_11_8f3a618f4bd64834d3636281b025ea9d) (Utilities|Struct|BreakFruitReppedData (Variables|GetFruitReppedData_OwningPlayerController_6_7553902549A37048686212BC13E7D8CA)))
  (CallFunction|UpdateGrowthStatus false)
  (CallFunction|SetLootTierGroupfromRarity)
  ;[1 1 1] NewStuff
  (Utilities|IsValid _fruitreppeddata_saplingparent_11_8f3a618f4bd64834d3636281b025ea9d
    (:Is%20Valid
      (Networking|SwitchHasAuthority
        (:Authority
          (Class|BSMAAthenaRealitySaplingParent|NotifyFruitStateChanged _fruitreppeddata_saplingparent_11_8f3a618f4bd64834d3636281b025ea9d self))
        (:Remote
          (Default|CallSaplingUpdateEvent _fruitreppeddata_saplingparent_11_8f3a618f4bd64834d3636281b025ea9d _fruitreppeddata_saplingparent_11_8f3a618f4bd64834d3636281b025ea9d))))))

(fn PlayLandingFX
  (params (bSpawnVFX :type Boolean :default true))
  (bind _staticmeshcomponent (Variables|Debug|GetStaticMeshComponent))
  (bind _returnvalue (Transformation|GetWorldTransform _staticmeshcomponent))
  (bind _location (.location _returnvalue))
  (Networking|CanExecuteCosmeticEvents
    (:True
      ;[1 1 1] Sound, FX, and Postprocess
      (if (not (Variables|Falling|GetLandInterrupted))
        (Audio|PlaySoundatLocation (Variables|Falling|GetSound_Landing) _location)
        (if (Variables|GetbSpawnVFX)
          (Effects|Components|ParticleSystem|SpawnFXAtLocation (Variables|Falling|GetLandingFX) _location))))))

(fn SetLootTierGroupFromRarity
  (|CallParentFunction)
  (switch Utilities|FlowControl|Switch|SwitchonSaplingFruitVariants (Variables|Default|GetVariantOrRarity)
    (:NewEnumerator0
      _)
    (:NewEnumerator1
      _)
    (:NewEnumerator9
      _)
    (:NewEnumerator2
      _)
    (:NewEnumerator3
      (Container|SetSearchLootTierGroup (Variables|Default|GetLootTierGroup_Nurtured1)))
    (:NewEnumerator4
      (Container|SetSearchLootTierGroup (Variables|Default|GetLootTierGroup_Nurtured2)))
    (:NewEnumerator5
      (Container|SetSearchLootTierGroup (Variables|Default|GetLootTierGroup_Nurtured3)))
    (:NewEnumerator6
      (Container|SetSearchLootTierGroup (Variables|Default|GetLootTierGroup_Nurtured4))
      (Variables|Default|SetIsGoldFruit true))
    (:NewEnumerator7
      _)
    (:NewEnumerator8
      (Container|SetSearchLootTierGroup (Variables|Default|GetLootTierGroup_Nurtured5))))
  (Container|SetSearchLootTierGroup (Variables|Default|GetLootTierGroup_Nurtured0)))

(fn SetToUseShortTimes
  (bind _growthtimer (Variables|Default|GetGrowthTimer))
  (Variables|Debug|SetUseShortTimes true)
  (if (Utilities|Time|IsValidTimerHandle _growthtimer)
    ;[1 1 1] Restart the lerp with a shorter timer if needed
    (if (> (Utilities|Time|GetTimerRemainingTimebyHandle _growthtimer) (ScalableFloat|GetValueatLevel (Variables|Default|GetShortTime)))
      (CallFunction|UpdateGrowthStatus false))))

(fn SetFruitScale
  (params (Scale :type Float%20%28double-precision%29))
  (bind _output_get (Variables|ColorandScale|SetCurr%20World%20Scale Scale))
  (Transformation|SetWorldScale3D (Variables|Debug|GetStaticMeshComponent) (Math|Vector|MakeVector _output_get _output_get _output_get)))

(fn SetFruitSizeAlpha
  (params (SizeAlpha :type Float%20%28double-precision%29))
  (bind _output_get (Variables|ColorandScale|SetCurr%20World%20Alpha SizeAlpha))
  (Rendering|Material|SetScalarParameterValue (Variables|Default|GetMaterialInstanceDynamic) "FruitGrowth" _output_get :!class /Script/Engine.MaterialInstanceDynamic))

(fn SetupBindToBuildingActor
  (params (Object :type Object%20Reference))
  (bind _asbuilding_actor (Utilities|Casting|CastToBuildingActor Object))
  (EventDispatchers|BindEventtoOnDied _asbuilding_actor (EventDispatchers|CreateEvent :!event OnLandingLocationChanged) :!class /Script/FortniteGame.BuildingActor))

(fn StartDecay
  (Variables|Interaction|SetbAllowInteract false)
  (Variables|ColorandScale|SetCur%20Max%20World%20Scale 0.01)
  (Variables|ColorandScale|SetCur%20Max%20World%20Alpha 0.01)
  (Variables|ColorandScale|SetDestInnerColor)
  (Variables|ColorandScale|SetDestOuterColor)
  (CallFunction|ResetLerpParamsandLerpTimer (Variables|ColorandScale|GetCurr%20World%20Scale) (Variables|ColorandScale|GetCurr%20World%20Alpha)))

(fn UpdateGrowthStatus
  (params (NewParam :type Boolean))
  (bind (_fruitreppeddata_owningplayercontroller_6_7553902549a37048686212bc13e7d8ca _fruitreppeddata_saplingparent_11_8f3a618f4bd64834d3636281b025ea9d) (Utilities|Struct|BreakFruitReppedData (Variables|GetFruitReppedData_OwningPlayerController_6_7553902549A37048686212BC13E7D8CA)))
  (bind _staticmeshcomponent (Variables|Debug|GetStaticMeshComponent))
  (bind (_fruitreppeddata_owningplayercontroller_6_7553902549a37048686212bc13e7d8ca_2 _fruitreppeddata_saplingparent_11_8f3a618f4bd64834d3636281b025ea9d_2) (Utilities|Struct|BreakFruitReppedData (Variables|GetFruitReppedData_OwningPlayerController_6_7553902549A37048686212BC13E7D8CA)))
  (bind (_fruitreppeddata_owningplayercontroller_6_7553902549a37048686212bc13e7d8ca_3 _fruitreppeddata_saplingparent_11_8f3a618f4bd64834d3636281b025ea9d_3) (Utilities|Struct|BreakFruitReppedData (Variables|GetFruitReppedData_OwningPlayerController_6_7553902549A37048686212BC13E7D8CA)))
  (bind _staticmeshcomponent_2 (Variables|Debug|GetStaticMeshComponent))
  (bind _staticmeshcomponent_3 (Variables|Debug|GetStaticMeshComponent))
  (bind _staticmeshcomponent_4 (Variables|Debug|GetStaticMeshComponent))
  (bind _staticmeshcomponent_5 (Variables|Debug|GetStaticMeshComponent))
  (bind _staticmeshcomponent_6 (Variables|Debug|GetStaticMeshComponent))
  (bind _staticmeshstem (Variables|Default|GetStaticMeshStem))
  ;[1 1 1] Don't change status if we're dying.
  (if (not (Variables|Default|GetScaleOutAndDie))
    (Utilities|FlowControl|Sequence
      (:then_0
        (switch Utilities|FlowControl|Switch|SwitchonSaplingFruitVariants (Variables|Default|GetVariantOrRarity)
          (:NewEnumerator0
            ;[1 1 1] Picked
            (Variables|Interaction|SetbAllowInteract false)
            (Variables|ColorandScale|SetCur%20Max%20World%20Scale (Variables|ColorandScale|GetMax%20World%20Scale%20Baby))
            (Variables|ColorandScale|SetCur%20Max%20World%20Alpha (Variables|ColorandScale|GetMax%20World%20Size%20Alpha%20Baby))
            (bind _outcolor (CallFunction|GetWeightedFruitColor (Variables|ColorConsts|GetDestColor_Baby_Inner) (ScalableFloat|GetValueatLevel (Variables|Default|GetBabyColorAlphaFactor)) true))
            (Variables|ColorandScale|SetDestInnerColor _outcolor)
            (bind _outcolor_2 (CallFunction|GetWeightedFruitColor (Variables|ColorConsts|GetDestColor_Baby_Outer) (ScalableFloat|GetValueatLevel (Variables|Default|GetBabyColorAlphaFactor))))
            (Variables|ColorandScale|SetDestOuterColor _outcolor_2)
            ;[1 1 1] Kill the stem when picked
            (Rendering|SetVisibility _staticmeshcomponent false false)
            (Collision|SetCollisionEnabled _staticmeshcomponent "NoCollision")
            (Rendering|SetVisibility _staticmeshstem false false)
            (Collision|SetCollisionEnabled _staticmeshstem "NoCollision")
            (bind _returnvalue (Utilities|Time|SetTimerbyFunctionName 0 "DestroyThisFruit" (CallFunction|GetScalableFloatTime (Variables|Default|GetPickedTimerLength))))
            (Variables|Default|SetGrowthTimer _returnvalue)
            (Utilities|Time|ClearTimerbyFunctionName 0 "Increment Growth State"))
          (:NewEnumerator1
            ;[1 1 1] Baby
            (Variables|Interaction|SetbAllowInteract false)
            (Variables|ColorandScale|SetCur%20Max%20World%20Scale (Variables|ColorandScale|GetMax%20World%20Scale%20Baby))
            (Variables|ColorandScale|SetCur%20Max%20World%20Alpha (Variables|ColorandScale|GetMax%20World%20Size%20Alpha%20Baby))
            (bind _outcolor_3 (CallFunction|GetWeightedFruitColor (Variables|ColorConsts|GetDestColor_Baby_Inner) (ScalableFloat|GetValueatLevel (Variables|Default|GetBabyColorAlphaFactor)) true))
            (Variables|ColorandScale|SetDestInnerColor _outcolor_3)
            (Variables|ColorandScale|SetCurInnerColor _outcolor_3)
            (bind _outcolor_4 (CallFunction|GetWeightedFruitColor (Variables|ColorConsts|GetDestColor_Baby_Outer) (ScalableFloat|GetValueatLevel (Variables|Default|GetBabyColorAlphaFactor))))
            (Variables|ColorandScale|SetDestOuterColor _outcolor_4)
            (Variables|ColorandScale|SetCurOuterColor _outcolor_4)
            (CallFunction|ResetLerpParamsandLerpTimer (Variables|ColorandScale|GetMin%20World%20Scale%20Baby) (Variables|ColorandScale|GetMin%20World%20Size%20Alpha%20Baby))
            (Rendering|SetVisibility _staticmeshcomponent_6 true false)
            (Collision|SetCollisionEnabled _staticmeshcomponent_6 "QueryOnly")
            (Collision|SetCollisionProfileName _staticmeshcomponent_6 "FortCameraOnly" true :!class /Script/Engine.PrimitiveComponent)
            (bind _returnvalue_2 (Utilities|Time|SetTimerbyFunctionName 0 "Increment Growth State" (CallFunction|GetScalableFloatTime (Variables|Default|GetBabyTimerLength))))
            (Variables|Default|SetGrowthTimer _returnvalue_2))
          (:NewEnumerator9
            ;[1 1 1] Young
            (Variables|Interaction|SetbAllowInteract false)
            (Variables|ColorandScale|SetCur%20Max%20World%20Scale (Variables|ColorandScale|GetMax%20World%20Scale%20Youth))
            (Variables|ColorandScale|SetCur%20Max%20World%20Alpha (Variables|ColorandScale|GetMax%20World%20Size%20Alpha%20Youth%20))
            (bind _outcolor_5 (CallFunction|GetWeightedFruitColor (Variables|ColorConsts|GetDestColor_Youth_Inner) (ScalableFloat|GetValueatLevel (Variables|Default|GetYouthColorAlphaFactor)) true))
            (Variables|ColorandScale|SetDestInnerColor _outcolor_5)
            (bind _outcolor_6 (CallFunction|GetWeightedFruitColor (Variables|ColorConsts|GetDestColor_Youth_Outer) (ScalableFloat|GetValueatLevel (Variables|Default|GetYouthColorAlphaFactor))))
            (Variables|ColorandScale|SetDestOuterColor _outcolor_6)
            (CallFunction|ResetLerpParamsandLerpTimer (Variables|ColorandScale|GetCurr%20World%20Scale) (Variables|ColorandScale|GetCurr%20World%20Alpha))
            (Class|BSMAAthenaRealitySaplingParent|EventGrowthSound _fruitreppeddata_saplingparent_11_8f3a618f4bd64834d3636281b025ea9d_3)
            (Collision|SetCollisionEnabled _staticmeshcomponent_4 "QueryOnly")
            (Collision|SetCollisionProfileName _staticmeshcomponent_4 "FortCameraOnly" true :!class /Script/Engine.PrimitiveComponent)
            (bind _returnvalue_3 (Utilities|Time|SetTimerbyFunctionName 0 "Increment Growth State" (CallFunction|GetScalableFloatTime (Variables|Default|GetYouthTimerLength))))
            (Variables|Default|SetGrowthTimer _returnvalue_3))
          (:NewEnumerator2
            _)
          (:NewEnumerator3
            _)
          (:NewEnumerator4
            _)
          (:NewEnumerator5
            _)
          (:NewEnumerator6
            _)
          (:NewEnumerator7
            (Utilities|FlowControl|Sequence
              (:then_0
                (bind _outcolor_7 (CallFunction|GetColorbyRarity (Variables|Default|GetVariantOrRarity) true :!label _label))
                (Variables|ColorandScale|SetDestInnerColor _outcolor_7)
                (Variables|ColorandScale|SetCurInnerColor _outcolor_7)
                (bind _outcolor_8 (CallFunction|GetColorbyRarity (Variables|Default|GetVariantOrRarity)))
                (Variables|ColorandScale|SetDestOuterColor _outcolor_8)
                (Variables|ColorandScale|SetCurOuterColor _outcolor_8))
              (:then_1
                ;[1 1 1] Mythic Baby
                (Variables|Interaction|SetbAllowInteract false)
                (Variables|ColorandScale|SetCur%20Max%20World%20Alpha (Variables|ColorandScale|GetMax%20World%20Size%20Alpha%20Baby))
                (Variables|ColorandScale|SetCur%20Max%20World%20Scale (Variables|ColorandScale|GetMax%20World%20Scale%20Baby))
                (CallFunction|ResetLerpParamsandLerpTimer (Variables|ColorandScale|GetMin%20World%20Scale%20Baby) (Variables|ColorandScale|GetMin%20World%20Size%20Alpha%20Baby))
                (Rendering|SetVisibility _staticmeshcomponent_3 true false)
                (Collision|SetCollisionEnabled _staticmeshcomponent_3 "QueryOnly"))))
          (:NewEnumerator8
            (Utilities|FlowControl|Sequence
              (:then_0
                (goto _label))
              (:then_1
                ;[1 1 1] Mature, Overripe, Rotten
                (Variables|Interaction|SetbAllowInteract true)
                (Variables|ColorandScale|SetCur%20Max%20World%20Scale (Variables|ColorandScale|GetMax%20World%20Scale%20Mythic%20Mature))
                (Collision|SetCollisionEnabled _staticmeshcomponent_2 "QueryOnly")
                (Collision|SetCollisionProfileName _staticmeshcomponent_2 "FortContainerOverlapPawns" true :!class /Script/Engine.PrimitiveComponent)
                (CallFunction|ResetLerpParamsandLerpTimer (Variables|ColorandScale|GetMax%20World%20Scale%20Mythic%20Mature) (Variables|ColorandScale|GetMax%20World%20Size%20Alpha%20Mythic%20Mature))
                (Variables|Container|SetSearchText "NSLOCTEXT(\"[1FE45CC64F0C83605A3B259422F66CB6]\", \"E5DFFEB247783DD97A04148DD95DCFA2\", \"Pick <Rarity_Mythic>Mythic</> Fruit\")")))))
        (Utilities|FlowControl|Sequence
          (:then_0
            (bind _outcolor_9 (CallFunction|GetColorbyRarity (Variables|Default|GetVariantOrRarity) true))
            (Variables|ColorandScale|SetDestInnerColor _outcolor_9)
            (bind _outcolor_10 (CallFunction|GetColorbyRarity (Variables|Default|GetVariantOrRarity)))
            (Variables|ColorandScale|SetDestOuterColor _outcolor_10))
          (:then_1
            _))
        ;[1 1 1] Mature (& decay if on the ground)
        (Variables|Interaction|SetbAllowInteract true)
        (Variables|ColorandScale|SetCur%20Max%20World%20Alpha (Variables|ColorandScale|GetMax%20World%20Size%20Alpha%20Mature))
        (Variables|ColorandScale|SetCur%20Max%20World%20Scale (Variables|ColorandScale|GetMax%20World%20Scale%20Mature))
        (Collision|SetCollisionEnabled _staticmeshcomponent_5 "QueryOnly")
        (Collision|SetCollisionProfileName _staticmeshcomponent_5 "FortContainerOverlapPawns" true :!class /Script/Engine.PrimitiveComponent)
        (CallFunction|ResetLerpParamsandLerpTimer (Variables|ColorandScale|GetCurr%20World%20Scale) (Variables|ColorandScale|GetCurr%20World%20Alpha))
        (Class|BSMAAthenaRealitySaplingParent|EventGrowthSound _fruitreppeddata_saplingparent_11_8f3a618f4bd64834d3636281b025ea9d)
        (Utilities|IsValid _fruitreppeddata_saplingparent_11_8f3a618f4bd64834d3636281b025ea9d_2
          (:Is%20Valid
            (LocalVariables|SetGrayAsInt (Math|Conversions|ToInteger(Byte) (Variables|Default|GetGray)))
            (if (not (>= (Math|Conversions|ToInteger(Byte) (Variables|Default|GetVariantOrRarity)) (+ (Variables|/RealitySeedGameplay/items/RealitySaplings/B_SMA_Athena_RealitySapling_Parent.B_SMA_Athena_RealitySapling_Parent_C|GetNumTimesWeeded _fruitreppeddata_saplingparent_11_8f3a618f4bd64834d3636281b025ea9d_2) (LocalVariables|GetGrayAsInt))))
              (Variables|Container|SetSearchText "NSLOCTEXT(\"[1FE45CC64F0C83605A3B259422F66CB6]\", \"340AFC7F4936E1B9D61C208635C78429\", \"Pick Fruit\")")
              (bind _returnvalue_4 (Utilities|Time|SetTimerbyFunctionName 0 "Increment Growth State" (CallFunction|GetScalableFloatTime (Variables|Default|GetMatureTimerLength))))
              (Variables|Default|SetGrowthTimer _returnvalue_4)))
          (:Is%20Not%20Valid
            _))
        (bind _text (CallFunction|GetFruitRarityRichText))
        (Variables|Container|SetSearchText (Utilities|Text|FormatText "NSLOCTEXT(\"[1FE45CC64F0C83605A3B259422F66CB6]\", \"C1D5A50A4FA06FCE7A9E7C909CCB7DF0\", \"Pick Fruit\\r\\n{Rarity}\")" _text)))
      (:then_1
        ;[1 1 1] Decay if on ground
        (switch Utilities|FlowControl|Switch|SwitchonSaplingFruitVariants (Variables|Default|GetVariantOrRarity)
          (:NewEnumerator1
            _)
          (:NewEnumerator9
            _)
          (:NewEnumerator2
            _)
          (:NewEnumerator3
            _)
          (:NewEnumerator4
            _)
          (:NewEnumerator5
            _)
          (:NewEnumerator6
            _))
        (if (!= (Variables|Falling|GetFallCount_repped) 0)
          (Utilities|Time|SetTimerbyFunctionName 0 "StartDecay" (- (ScalableFloat|GetValueatLevel (Variables|Default|GetDelayTimerLength)) 1.0))))
      (:then_2
        ;[0.4155 0.3973 1] SetCanBeDamaged to whether we're allowing interact. (Don't allow growing fruit to be damaged)
        (Variables|Actor|SetbCanBeDamaged (Variables|Interaction|GetbAllowInteract))))))

(fn ConstructionScript
  (|CallParentFunction))
