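; Mythic Reality Sapling fruit: a lootable container that only the sapling's owning player can search, with an exception for persistence overrides.
; Once playlist data loads, the server sets a replicated init state from the all-players setting.
; Clients then use that state, or squad membership and setting, to show or hide the fruit.
; Loot tier follows the biome at the sapling's location, and ShouldDie blocks non-owner kills to prevent stealing.
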(event EventBeginPlay
  (Variables|Default|SetIsMythicFruit true)
  (|CallParentFunction)
  (Networking|SwitchHasAuthority
    (:Authority
      (Variables|Default|SetbAvailableLocally true)
      (Class|FortAsyncActionCurrentPlaylistReady|PlaylistDataReadyAsync
        (:OnReady
          (Variables|Default|SetPlaylistDataReady true)
          (Default|InitializeAllPlayersCanInteractandSeeState))))))

(event Collision|EventActorBeginOverlap
  (params (OtherActor :type Actor%20Object%20Reference))
  (|CallParentFunction OtherActor) ; This node is disabled and will not be called.\nDrag off pins to build functionality.
)

(event EventOnFruitLooted
  (params (FortPawn :type Fort%20Pawn%20Object%20Reference))
  (|CallParentFunction 0)
  (Components|DestroyComponent (Variables|Default|GetNS_Idle)))

(event EventOnSaplingKilled
  (Components|DestroyComponent (Variables|Default|GetNS_Idle)))

(fn BlueprintCanInteract
  (params (InteractingPawn :type Fort%20Pawn%20Object%20Reference) (InteractionBeingAttempted :type EInteractionBeingAttempted%20Enum :default "FirstInteraction") (InteractionType :type TInteractionType%20Enum :default "IT_NoInteraction"))
  (bind _fruitreppeddata_owningplayercontroller_6_7553902549a37048686212bc13e7d8ca (Utilities|Struct|BreakFruitReppedData (Variables|GetFruitReppedData_OwningPlayerController_6_7553902549A37048686212BC13E7D8CA)))
  (bind _returnvalue (== InteractingPawn _fruitreppeddata_owningplayercontroller_6_7553902549a37048686212bc13e7d8ca))
  (bind (_fruitreppeddata_owningplayercontroller_6_7553902549a37048686212bc13e7d8ca_2 _fruitreppeddata_saplingparent_11_8f3a618f4bd64834d3636281b025ea9d) (Utilities|Struct|BreakFruitReppedData (Variables|GetFruitReppedData_OwningPlayerController_6_7553902549A37048686212BC13E7D8CA)))
  (bind _owning_player_controller_repped (Variables|/RealitySeedGameplay/items/RealitySaplings/B_SMA_Athena_RealitySapling_Parent.B_SMA_Athena_RealitySapling_Parent_C|GetOwning%20Player%20Controller%20Repped _fruitreppeddata_saplingparent_11_8f3a618f4bd64834d3636281b025ea9d))
  ;[1 1 1] Only allow the owning player to interact.
  (if (|CallParentFunction InteractingPawn InteractionBeingAttempted InteractionType)
    (Utilities|IsValid _fruitreppeddata_saplingparent_11_8f3a618f4bd64834d3636281b025ea9d
      (:Is%20Valid
        (Utilities|IsValid _owning_player_controller_repped
          (:Is%20Valid
            (return (== (Pawn|GetController InteractingPawn) _owning_player_controller_repped)))
          (:Is%20Not%20Valid
            (if (and (Variables|/RealitySeedGameplay/items/RealitySaplings/B_SMA_Athena_RealitySapling_Parent.B_SMA_Athena_RealitySapling_Parent_C|GetOverrride%20Sapling%20Persistence _fruitreppeddata_saplingparent_11_8f3a618f4bd64834d3636281b025ea9d) (Variables|/RealitySeedGameplay/items/RealitySaplings/B_SMA_Athena_RealitySapling_Parent.B_SMA_Athena_RealitySapling_Parent_C|GetOverrride%20Sapling%20Persistence _fruitreppeddata_saplingparent_11_8f3a618f4bd64834d3636281b025ea9d))
              (return true)
              (else
                (return false))))))
      (:Is%20Not%20Valid
        (if _returnvalue
          (return _returnvalue)
          (else
            (Utilities|FlowControl|Sequence
              (:then_1
                (return false)))))))))

(fn BlueprintGetFailedInteractionString
  (params (InteractingPawn :type Fort%20Pawn%20Object%20Reference) (InteractionBeingAttempted :type EInteractionBeingAttempted%20Enum :default "FirstInteraction"))
  (return "NSLOCTEXT(\"[2077607A4CEF310ADCCBEE8411172E83]\", \"0E6ED29C4AEB601FB7950F97EFE7B8EC\", \"This is not your fruit.\")"))

(fn BlueprintGetInteractionString
  (params (InteractingPawn :type Fort%20Pawn%20Object%20Reference) (InteractionBeingAttempted :type EInteractionBeingAttempted%20Enum :default "FirstInteraction"))
  (bind _returnvalue (CallFunction|IsOwningPlayerPawn InteractingPawn))
  (if _returnvalue
    (return (Variables|Container|GetSearchText))
    (else
      (return "NSLOCTEXT(\"[2077607A4CEF310ADCCBEE8411172E83]\", \"0910F2BE417DFE939183AFBD9CDC7631\", \"This is not your fruit.\")"))))

(fn BlueprintOnLocalInteract
  (params (InteractingPawn :type Fort%20Player%20Pawn%20Object%20Reference) (InteractionBeingAttempted :type EInteractionBeingAttempted%20Enum :default "FirstInteraction"))
  (bind _returnvalue (CallFunction|IsOwningPlayerPawn InteractingPawn))
  (return _returnvalue))

(fn FruitDataRepped
  (bind (_fruitreppeddata_owningplayercontroller_6_7553902549a37048686212bc13e7d8ca _fruitreppeddata_saplingparent_11_8f3a618f4bd64834d3636281b025ea9d) (Utilities|Struct|BreakFruitReppedData (Variables|GetFruitReppedData_OwningPlayerController_6_7553902549A37048686212BC13E7D8CA)))
  (bind (_fruitreppeddata_owningplayercontroller_6_7553902549a37048686212bc13e7d8ca_2 _fruitreppeddata_saplingparent_11_8f3a618f4bd64834d3636281b025ea9d_2 _fruitreppeddata_initializationstate_13_ce95c12a48b098940bc3759d2370592d) (Utilities|Struct|BreakFruitReppedData (Variables|GetFruitReppedData_OwningPlayerController_6_7553902549A37048686212BC13E7D8CA)))
  (|CallParentFunction)
  (if (Utilities|IsValid _fruitreppeddata_saplingparent_11_8f3a618f4bd64834d3636281b025ea9d :!class /Script/Engine.KismetSystemLibrary)
    (if (>= _fruitreppeddata_initializationstate_13_ce95c12a48b098940bc3759d2370592d 0)
      (Default|InitializefromLocalPlayerControllerandDetermineEnabledForLocalPlayer)
      (else
        (Default|InitializeAllPlayersCanInteractandSeeState)))))

(fn IncrementGrowthState)

(fn InitializeAllPlayersCanInteractAndSeeState
  (bind (_fruitreppeddata_owningplayercontroller_6_7553902549a37048686212bc13e7d8ca _fruitreppeddata_saplingparent_11_8f3a618f4bd64834d3636281b025ea9d _fruitreppeddata_initializationstate_13_ce95c12a48b098940bc3759d2370592d _fruitreppeddata_overrideislocalplayersapling_17_74dc52a84db6ad597d0dbb8dfccba192) (Utilities|Struct|BreakFruitReppedData (Variables|GetFruitReppedData_OwningPlayerController_6_7553902549A37048686212BC13E7D8CA)))
  (bind (_fruitreppeddata_owningplayercontroller_6_7553902549a37048686212bc13e7d8ca_2 _fruitreppeddata_saplingparent_11_8f3a618f4bd64834d3636281b025ea9d_2 _fruitreppeddata_initializationstate_13_ce95c12a48b098940bc3759d2370592d_2) (Utilities|Struct|BreakFruitReppedData (Variables|GetFruitReppedData_OwningPlayerController_6_7553902549A37048686212BC13E7D8CA)))
  (bind (_fruitreppeddata_owningplayercontroller_6_7553902549a37048686212bc13e7d8ca_3 _fruitreppeddata_saplingparent_11_8f3a618f4bd64834d3636281b025ea9d_3 _fruitreppeddata_initializationstate_13_ce95c12a48b098940bc3759d2370592d_3 _fruitreppeddata_overrideislocalplayersapling_17_74dc52a84db6ad597d0dbb8dfccba192_2) (Utilities|Struct|BreakFruitReppedData (Variables|GetFruitReppedData_OwningPlayerController_6_7553902549A37048686212BC13E7D8CA)))
  (Networking|SwitchHasAuthority
    (:Authority
      (if (Math|Boolean|ANDBoolean (Variables|Default|GetPlaylistDataReady) (Utilities|IsValid _fruitreppeddata_saplingparent_11_8f3a618f4bd64834d3636281b025ea9d_2 :!class /Script/Engine.KismetSystemLibrary) (<= _fruitreppeddata_initializationstate_13_ce95c12a48b098940bc3759d2370592d_2 -1))
        (if (ScalableFloat|GetValueasBool (Variables|Default|GetEnabledForAllPlayers))
          (Variables|SetFruitReppedData_OwningPlayerController_6_7553902549A37048686212BC13E7D8CA (Utilities|Struct|MakeFruitReppedData _fruitreppeddata_owningplayercontroller_6_7553902549a37048686212bc13e7d8ca _fruitreppeddata_saplingparent_11_8f3a618f4bd64834d3636281b025ea9d 1 _fruitreppeddata_overrideislocalplayersapling_17_74dc52a84db6ad597d0dbb8dfccba192))
          (else
            (Variables|SetFruitReppedData_OwningPlayerController_6_7553902549A37048686212BC13E7D8CA (Utilities|Struct|MakeFruitReppedData _fruitreppeddata_owningplayercontroller_6_7553902549a37048686212bc13e7d8ca_3 _fruitreppeddata_saplingparent_11_8f3a618f4bd64834d3636281b025ea9d_3 0 _fruitreppeddata_overrideislocalplayersapling_17_74dc52a84db6ad597d0dbb8dfccba192_2))))))))

(fn InitializeFromLocalPlayerControllerAndDetermineEnabledforLocalPlayer
  (bind _fruitreppeddata_owningplayercontroller_6_7553902549a37048686212bc13e7d8ca (Utilities|Struct|BreakFruitReppedData (Variables|GetFruitReppedData_OwningPlayerController_6_7553902549A37048686212BC13E7D8CA)))
  (bind _returnvalue (Game|GetLocalPlayerControllerID _fruitreppeddata_owningplayercontroller_6_7553902549a37048686212bc13e7d8ca))
  (bind _fruitreppeddata_owningplayercontroller_6_7553902549a37048686212bc13e7d8ca_2 (Utilities|Struct|BreakFruitReppedData (Variables|GetFruitReppedData_OwningPlayerController_6_7553902549A37048686212BC13E7D8CA)))
  (bind (_fruitreppeddata_owningplayercontroller_6_7553902549a37048686212bc13e7d8ca_3 _fruitreppeddata_saplingparent_11_8f3a618f4bd64834d3636281b025ea9d) (Utilities|Struct|BreakFruitReppedData (Variables|GetFruitReppedData_OwningPlayerController_6_7553902549A37048686212BC13E7D8CA)))
  (bind (_fruitreppeddata_owningplayercontroller_6_7553902549a37048686212bc13e7d8ca_4 _fruitreppeddata_saplingparent_11_8f3a618f4bd64834d3636281b025ea9d_2 _fruitreppeddata_initializationstate_13_ce95c12a48b098940bc3759d2370592d _fruitreppeddata_overrideislocalplayersapling_17_74dc52a84db6ad597d0dbb8dfccba192) (Utilities|Struct|BreakFruitReppedData (Variables|GetFruitReppedData_OwningPlayerController_6_7553902549A37048686212BC13E7D8CA)))
  (bind (_fruitreppeddata_owningplayercontroller_6_7553902549a37048686212bc13e7d8ca_5 _fruitreppeddata_saplingparent_11_8f3a618f4bd64834d3636281b025ea9d_3 _fruitreppeddata_initializationstate_13_ce95c12a48b098940bc3759d2370592d_2) (Utilities|Struct|BreakFruitReppedData (Variables|GetFruitReppedData_OwningPlayerController_6_7553902549A37048686212BC13E7D8CA)))
  (if (Variables|Debug|GetDebugFruit)
    ;[1 1 1] Debug prints
    (Utilities|FlowControl|Sequence
      (:then_0
        (Development|PrintString (Utilities|String|Append "controller:" (Utilities|GetDisplayName _fruitreppeddata_owningplayercontroller_6_7553902549a37048686212bc13e7d8ca_4) "   sapling parent:" (Utilities|GetDisplayName _fruitreppeddata_saplingparent_11_8f3a618f4bd64834d3636281b025ea9d_2) "   init state:" (Utilities|String|ToString(Integer) _fruitreppeddata_initializationstate_13_ce95c12a48b098940bc3759d2370592d) "   isLocalPlayer:" (Utilities|String|ToString(Boolean) _fruitreppeddata_overrideislocalplayersapling_17_74dc52a84db6ad597d0dbb8dfccba192)) true true "(R=0.000000,G=0.660000,B=1.000000,A=1.000000)" 2e+01))
      (:then_1
        _)))
  (if (or (== _fruitreppeddata_initializationstate_13_ce95c12a48b098940bc3759d2370592d_2 1) (and (Team|OnSameTeam (Variables|/RealitySeedGameplay/items/RealitySaplings/B_SMA_Athena_RealitySapling_Parent.B_SMA_Athena_RealitySapling_Parent_C|GetOwning%20Player%20State%20Repped _fruitreppeddata_saplingparent_11_8f3a618f4bd64834d3636281b025ea9d_3) (Fortnite|GetFirstLocalFortPlayerController true)) (ScalableFloat|GetValueasBool (Variables|Default|GetEnabledForSquad))))
    (Variables|Default|SetbAvailableLocally true)
    (Components|Activation|Activate (Variables|Default|GetNS_Idle) false)
    (else
      ;[1 1 1] If this isn't owned by the local player hide it and no interacting with it.
      (Utilities|IsValid _fruitreppeddata_owningplayercontroller_6_7553902549a37048686212bc13e7d8ca_2
        (:Is%20Valid
          (Development|PrintString (Utilities|String|Append "Owning player controller repped is: " (Utilities|GetDisplayName _fruitreppeddata_owningplayercontroller_6_7553902549a37048686212bc13e7d8ca) "     LocalPlayerController ID:" (Utilities|String|ToString(Integer) _returnvalue)))
          (if (!= _returnvalue -1)
            (Variables|Default|SetbAvailableLocally true)
            (else
              _)))
        (:Is%20Not%20Valid
          (Utilities|IsValid _fruitreppeddata_saplingparent_11_8f3a618f4bd64834d3636281b025ea9d
            (:Is%20Valid
              (if (Variables|/RealitySeedGameplay/items/RealitySaplings/B_SMA_Athena_RealitySapling_Parent.B_SMA_Athena_RealitySapling_Parent_C|GetOverrride%20Sapling%20Persistence _fruitreppeddata_saplingparent_11_8f3a618f4bd64834d3636281b025ea9d)
                _
                (else
                  _)))
            (:Is%20Not%20Valid
              (Fortnite|FortShippingLog 0 (Utilities|String|Append (Utilities|GetDisplayName self) " does not have a valid Sapling owner at initialization "))))))
      ;[1 1 1] hide the sapling
      (Variables|Default|SetbAvailableLocally false)
      (Rendering|SetVisibility (Variables|Debug|GetStaticMeshComponent) (Variables|Default|GetbAvailableLocally) true)
      (CallFunction|SetSuppressHealthBar (not (Variables|Default|GetbAvailableLocally)))
      (Components|DestroyComponent (Variables|Default|GetNS_Idle)))))

(fn IsOwningPlayerPawn
  (params (InteractingPawn :type Fort%20Pawn%20Object%20Reference))
  (bind _fruitreppeddata_owningplayercontroller_6_7553902549a37048686212bc13e7d8ca (Utilities|Struct|BreakFruitReppedData (Variables|GetFruitReppedData_OwningPlayerController_6_7553902549A37048686212BC13E7D8CA)))
  (bind (_fruitreppeddata_owningplayercontroller_6_7553902549a37048686212bc13e7d8ca_2 _fruitreppeddata_saplingparent_11_8f3a618f4bd64834d3636281b025ea9d) (Utilities|Struct|BreakFruitReppedData (Variables|GetFruitReppedData_OwningPlayerController_6_7553902549A37048686212BC13E7D8CA)))
  (bind _owning_player_controller_repped (Variables|/RealitySeedGameplay/items/RealitySaplings/B_SMA_Athena_RealitySapling_Parent.B_SMA_Athena_RealitySapling_Parent_C|GetOwning%20Player%20Controller%20Repped _fruitreppeddata_saplingparent_11_8f3a618f4bd64834d3636281b025ea9d))
  (Utilities|IsValid _fruitreppeddata_saplingparent_11_8f3a618f4bd64834d3636281b025ea9d
    (:Is%20Valid
      (Utilities|IsValid _owning_player_controller_repped
        (:Is%20Valid
          (return (== (Pawn|GetController InteractingPawn) _owning_player_controller_repped)))
        (:Is%20Not%20Valid
          (if (and (Variables|/RealitySeedGameplay/items/RealitySaplings/B_SMA_Athena_RealitySapling_Parent.B_SMA_Athena_RealitySapling_Parent_C|GetOverrride%20Sapling%20Persistence _fruitreppeddata_saplingparent_11_8f3a618f4bd64834d3636281b025ea9d) (Variables|/RealitySeedGameplay/items/RealitySaplings/B_SMA_Athena_RealitySapling_Parent.B_SMA_Athena_RealitySapling_Parent_C|GetOverrideIsLocalPlayerSapling _fruitreppeddata_saplingparent_11_8f3a618f4bd64834d3636281b025ea9d))
            (return true)
            (else
              (return false))))))
    (:Is%20Not%20Valid
      (Utilities|IsValid _fruitreppeddata_owningplayercontroller_6_7553902549a37048686212bc13e7d8ca
        (:Is%20Valid
          (return (== InteractingPawn _fruitreppeddata_owningplayercontroller_6_7553902549a37048686212bc13e7d8ca)))
        (:Is%20Not%20Valid
          (return false))))))

(fn SetLootTierGroupFromRarity
  (bind _returnvalue (Fortnite|Athena|GameState|GetGameStateAthena))
  (bind (_fruitreppeddata_owningplayercontroller_6_7553902549a37048686212bc13e7d8ca _fruitreppeddata_saplingparent_11_8f3a618f4bd64834d3636281b025ea9d) (Utilities|Struct|BreakFruitReppedData (Variables|GetFruitReppedData_OwningPlayerController_6_7553902549A37048686212BC13E7D8CA)))
  (bind (_persistentsaplingdata_owningplayercontroller _persistentsaplingdata_location) (Utilities|Struct|BreakRealitySaplingPersistentData (Variables|/Script/RealitySeedGameplayRuntime.BuildingRealitySapling|GetPersistentSaplingData_OwningPlayerController _fruitreppeddata_saplingparent_11_8f3a618f4bd64834d3636281b025ea9d)))
  (bind (_fruitreppeddata_owningplayercontroller_6_7553902549a37048686212bc13e7d8ca_2 _fruitreppeddata_saplingparent_11_8f3a618f4bd64834d3636281b025ea9d_2) (Utilities|Struct|BreakFruitReppedData (Variables|GetFruitReppedData_OwningPlayerController_6_7553902549A37048686212BC13E7D8CA)))
  (|CallParentFunction)
  (Utilities|IsValid _returnvalue
    (:Is%20Valid
      (Utilities|IsValid _fruitreppeddata_saplingparent_11_8f3a618f4bd64834d3636281b025ea9d_2
        (:Is%20Valid
          (LocalVariables|SetUseLocation _persistentsaplingdata_location))
        (:Is%20Not%20Valid
          (LocalVariables|SetUseLocation (Transformation|GetActorLocation))))
      (bind _returnvalue_2 (POI|GetPoiGridTagsforLocation _returnvalue (LocalVariables|GetUseLocation)))
      (if (GameplayTags|HasTag _returnvalue_2 "(TagName=\"Athena.Location.Biome.Arid\")")
        (Container|SetSearchLootTierGroup (Variables|Default|GetLoot_RealityFruit_Mythic_Desert))
        (elif (GameplayTags|HasTag _returnvalue_2 "(TagName=\"Athena.Location.Biome.Mountain\")")
          (Container|SetSearchLootTierGroup (Variables|Default|GetLoot_RealityFruit_Mythic_Mountain))
          (elif (GameplayTags|HasTag _returnvalue_2 "(TagName=\"Athena.Location.Biome.Reality\")")
            (Container|SetSearchLootTierGroup (Variables|Default|GetLoot_RealityFruit_Mythic_Reality))
            (else
              (Container|SetSearchLootTierGroup (Variables|Default|GetLoot_RealityFruit_Mythic_Default)))))))
    (:Is%20Not%20Valid
      (Container|SetSearchLootTierGroup (Variables|Default|GetLoot_RealityFruit_Mythic_Default)))))

(fn ShouldDie
  (params (Damage :type Float%20%28single-precision%29) (EventInstigator :type Controller%20Object%20Reference) (DamageCauser :type Actor%20Object%20Reference) (bIsTest :type Boolean) (EffectContext :type Gameplay%20Effect%20Context%20Handle%20Structure%20%28by%20ref%29))
  (bind _fruitreppeddata_owningplayercontroller_6_7553902549a37048686212bc13e7d8ca (Utilities|Struct|BreakFruitReppedData (Variables|GetFruitReppedData_OwningPlayerController_6_7553902549A37048686212BC13E7D8CA)))
  ;[1 0.7104 0.4004] If the fruit has an owner, and the owner is NOT the instigator, then don't die. This prevents players from "stealing" a player's mythic by shooting it (even if they can't see or interact with it)
  (Utilities|IsValid _fruitreppeddata_owningplayercontroller_6_7553902549a37048686212bc13e7d8ca
    (:Is%20Valid
      (if (!= _fruitreppeddata_owningplayercontroller_6_7553902549a37048686212bc13e7d8ca EventInstigator)
        (return false)
        (else
          _)))
    (:Is%20Not%20Valid
      _))
  (bind _returnvalue (|CallParentFunction Damage 0 DamageCauser bIsTest EffectContext))
  (return _returnvalue))

(fn ShouldDisplayDamageNumbers
  (return false))

(fn ConstructionScript
  (|CallParentFunction))
