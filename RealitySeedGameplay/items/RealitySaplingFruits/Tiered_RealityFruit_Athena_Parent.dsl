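; Base Blueprint for Reality Sapling fruit, a chest-style lootable container grown on a parent sapling.
; At BeginPlay the sapling's weeding count sets the max tier, then calls an empty rarity loot-group hook for children.
; Interaction requires a living sapling; interaction and death are owner-restricted for Outpost controllers.
; Cosmetic-only paths drive audio cues, mobile MID wiggle, outline and icon, and a one-shot hide and material-timestamp opening on loot.
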(event EventBeginPlay
  (bind (_fruitreppeddata_owningplayercontroller_6_7553902549a37048686212bc13e7d8ca _fruitreppeddata_saplingparent_11_8f3a618f4bd64834d3636281b025ea9d) (Utilities|Struct|BreakFruitReppedData (Variables|GetFruitReppedData_OwningPlayerController_6_7553902549A37048686212BC13E7D8CA)))
  ;[1 1 1] Begin Play
  (Utilities|IsValid _fruitreppeddata_saplingparent_11_8f3a618f4bd64834d3636281b025ea9d
    (:Is%20Valid
      (CallFunction|SetLootTierGroupfromNurturing _fruitreppeddata_saplingparent_11_8f3a618f4bd64834d3636281b025ea9d))
    (:Is%20Not%20Valid
      _))
  (Networking|CanExecuteCosmeticEvents
    (:True
      (CallFunction|CreateMobileMIDs)
      (if (Variables|Container|GetbAlreadySearched)
        (CallFunction|SetLightVisibility false)
        (Audio|Components|Audio|Stop (Variables|Default|GetChest_Ambient_Sound))
        (else
          (CallFunction|SetLightVisibility true)
          (if (Variables|Default|GetEnableAudioIndicator)
            (CallFunction|StartAudiovisualIndicator (Variables|Default|GetChest_Ambient_Sound) "Chest" (Variables|Default|GetChestChimeVisualUpdate))))))))

(event Custom|Mobile_OnInteract
  ;[1 1 1] Mobile Wiggle on Interact
  (Networking|CanExecuteCosmeticEvents
    (:True
      (bind (_direction _lerp) (Timeline|MobileOnInteractTL Play)
        (:Update
          (for _array_element (Variables|Default|GetMIDs)
            (Rendering|Material|SetScalarParameterValue _array_element "WiggleOnInteractAmount" (Math|Float|Lerp 0.0 (Variables|Default|GetMobileWiggleAmount) _lerp) :!class /Script/Engine.MaterialInstanceDynamic)))))))

(event Custom|Mobile_Outline_On
  ;[1 1 1] Mobile Outline
  (for _array_element (Variables|Default|GetMIDs)
    (Rendering|Material|SetScalarParameterValue _array_element "Outline_Control" 1.0 :!class /Script/Engine.MaterialInstanceDynamic)
    (Rendering|Material|SetScalarParameterValue _array_element "Pulse_Control" 1.0 :!class /Script/Engine.MaterialInstanceDynamic)
    (Rendering|Material|SetVectorParameterValue _array_element "Color" "(R=1.000000,G=1.000000,B=1.000000,A=1.000000)" :!class /Script/Engine.MaterialInstanceDynamic)))

(event Container|EventEnableBacchusHighlight
  ;[1 1 1] Mobile highlighting
  (if (not (Variables|Container|GetbAlreadySearched))
    (Networking|CanExecuteCosmeticEvents
      (:True
        (CallFunction|SetVisibleMobileInteractIcon true)
        (bind (_direction _lerpobject _lerpinteactoicon) (Timeline|MobileSelectedTL Play)
          (:Update
            (bind _mobileinteracticon (Variables|Default|GetMobileInteractIcon)
              (:then
                (Transformation|SetRelativeScale3D _mobileinteracticon (Math|Vector|Lerp(Vector) "0, 0, 0" (Variables|Default|GetMobileInteractIconScale) _lerpinteactoicon)))
              (:else)))
          (:Finished
            (switch Utilities|FlowControl|Switch|SwitchonETimelineDirection _direction
              (:Backward
                (CallFunction|SetVisibleMobileInteractIcon false)))))))))

(event Custom|Mobile_OnDisconnect
  ;[1 1 1] Mobile Wiggle on Interact
  (Networking|CanExecuteCosmeticEvents
    (:True
      (Timeline|MobileOnInteractTL Reverse))))

(event Custom|Mobile_Outline_Off
  ;[1 1 1] Mobile Outline
  (for _array_element (Variables|Default|GetMIDs)
    (Rendering|Material|SetScalarParameterValue _array_element "Outline_Control" 0.0 :!class /Script/Engine.MaterialInstanceDynamic)
    (Rendering|Material|SetScalarParameterValue _array_element "Pulse_Control" 0.0 :!class /Script/Engine.MaterialInstanceDynamic)))

(event Container|EventDisableBacchusHighlight
  ;[1 1 1] Mobile highlighting
  (if (TouchPlatform|IsMobileGame)
    (Networking|CanExecuteCosmeticEvents
      (:True
        (Timeline|MobileSelectedTL Reverse)))))

(event EventOnBeginSearch
  ;[1 1 1] Mobile highlighting
  (if (Variables|Container|GetbAlreadySearched)
    _
    (elif (TouchPlatform|IsMobileGame)
      (Networking|CanExecuteCosmeticEvents
        (:True
          (Timeline|MobileSelectedTL Reverse))))))

(event EventOnSetSearched
  (Networking|CanExecuteCosmeticEvents
    (:True
      (CallFunction|HideChestOpenableCosmetics)
      (Development|PrintString "Searched"))))

(event EventDestroyed
  (Development|PrintString "Destroyed"))

(event EventOnLoot
  (CallFunction|OnFruitLooted (Variables|Default|GetLooting%20Pawn)))

(event Custom|OnFruitLooted
  (params (FortPawn :type Fort%20Pawn%20Object%20Reference))
  (Development|PrintString "Fruit Looted" true true "(R=1.000000,G=0.000000,B=0.049922,A=1.000000)" 2e+01)
  (Networking|CanExecuteCosmeticEvents
    (:True
      (Utilities|FlowControl|Sequence
        (:then_0
          (CallFunction|HideChestOpenableCosmetics))
        (:then_1
          (CallFunction|DetermineVFXRotation))
        (:then_2
          (CallFunction|SpawnChestOpenVFX))
        (:then_3
          (CallFunction|PlayChestOpeningTimeline))))
    (:False
      (Variables|Building|SetbDestroyOnPlayerBuildingPlacement true) ; Allows players to build over the opened chest mesh (but not when there is loot in there)
    )))

(event Custom|HideChestOpenableCosmetics
  (Utilities|FlowControl|DoOnce false
    (:Completed
      (CallFunction|SetLightVisibility false)
      (Audio|Components|Audio|FadeOut (Variables|Default|GetChest_Ambient_Sound) 0.5 0.0 "Linear" :!class /Script/Engine.AudioComponent)
      (CallFunction|StopAudiovisualIndicator)
      (CallFunction|CleanupWiggleMIDs)
      (Components|Timeline|Stop (Variables|Tiered_RealityFruit_Athena_Parent|GetMobileSelectedTL))
      (CallFunction|SetVisibleMobileInteractIcon false)
      (Rendering|SetRenderCustomDepth (Variables|Debug|GetStaticMeshComponent) false)
      (Audio|PlaySoundatLocation (Variables|Sound|GetOpen%20Chest%20Sound) (Transformation|GetActorLocation)))))

(event Custom|SpawnChestOpenVFX)

(event Custom|DetermineVFXRotation
  ;[0.926 0 1 1 80] FOR GALILEO ONLY
  (bind _returnvalue (Container|GetLootTier))
  (switch int _returnvalue
    (:Default
      (Variables|Default|SetOpenFX_Rotation -9e+01) ; Default chest
    )
    (:3
      (Variables|Default|SetP_Loot_Chest_Opened_FX (Variables|Default|GetGalileo_Chest_Open_FX)) ; 3 == Galileo Chest
      (Variables|Default|SetOpenFX_Rotation 0.0))))

(event Custom|PlayChestOpeningTimeline
  (Utilities|FlowControl|DoOnce false
    (:Completed
      (bind (_direction _time) (Timeline|Timeline_0 Play)
        (:Update
          (Rendering|Material|SetScalarParameterValueonMaterials (Variables|Debug|GetStaticMeshComponent) "CODE_Timestamp" _time))))))

(event EventOnDeathPlayEffects
  (params (Damage :type Float%20%28single-precision%29) (DamageTags :type Gameplay%20Tag%20Container%20Structure%20%28by%20ref%29) (Momentum :type Vector :default "0, 0, 0") (HitInfo :type Hit%20Result%20Structure%20%28by%20ref%29) (InstigatedBy :type Fort%20Pawn%20Object%20Reference) (DamageCauser :type Actor%20Object%20Reference) (EffectContext :type Gameplay%20Effect%20Context%20Handle%20Structure))
  (Audio|Components|Audio|FadeOut (Variables|Default|GetChest_Ambient_Sound) 0.5 0.0 "Linear" :!class /Script/Engine.AudioComponent)
  (Development|PrintString (Utilities|String|Append "Play Death Effects:" "instigated by:" (Utilities|GetDisplayName InstigatedBy) "   Damage Causer:" (Utilities|GetDisplayName DamageCauser)))
  (bind _asfort_player_pawn (Utilities|Casting|CastToFortPlayerPawn InstigatedBy))
  (Variables|Default|SetLooting%20Pawn _asfort_player_pawn))

(event Interaction|EventBlueprintOnBeginInteract
  (params (InteractingPawn :type Fort%20Pawn%20Object%20Reference) (InteractionBeingAttempted :type EInteractionBeingAttempted%20Enum :default "FirstInteraction") (InteractionType :type TInteractionType%20Enum :default "IT_NoInteraction"))
  (Variables|Default|SetLooting%20Pawn InteractingPawn))

(fn BlueprintCanInteract
  (params (InteractingPawn :type Fort%20Pawn%20Object%20Reference) (InteractionBeingAttempted :type EInteractionBeingAttempted%20Enum :default "FirstInteraction") (InteractionType :type TInteractionType%20Enum :default "IT_NoInteraction"))
  (bind (_fruitreppeddata_owningplayercontroller_6_7553902549a37048686212bc13e7d8ca _fruitreppeddata_saplingparent_11_8f3a618f4bd64834d3636281b025ea9d) (Utilities|Struct|BreakFruitReppedData (Variables|GetFruitReppedData_OwningPlayerController_6_7553902549A37048686212BC13E7D8CA)))
  (Utilities|IsValid _fruitreppeddata_saplingparent_11_8f3a618f4bd64834d3636281b025ea9d
    (:Is%20Valid
      (if (or (Variables|/RealitySeedGameplay/items/RealitySaplings/B_SMA_Athena_RealitySapling_Parent.B_SMA_Athena_RealitySapling_Parent_C|GetDestroyedForTransplant _fruitreppeddata_saplingparent_11_8f3a618f4bd64834d3636281b025ea9d) (Variables|/RealitySeedGameplay/items/RealitySaplings/B_SMA_Athena_RealitySapling_Parent.B_SMA_Athena_RealitySapling_Parent_C|GetDestroyedFromAging _fruitreppeddata_saplingparent_11_8f3a618f4bd64834d3636281b025ea9d))
        (Development|PrintString "Sapling fruit interaction disabled due to the parent sapling being destroyed (for transplant or aging)")
        (else
          (bind _asfort_player_controller_outpost (Utilities|Casting|CastToFortPlayerControllerOutpost (Pawn|GetController InteractingPawn))
            (:then
              (return (Outpost|IsOutpostOwner _asfort_player_controller_outpost)))
            (:CastFailed
              (return true))))))
    (:Is%20Not%20Valid
      _))
  (return false))

(fn CleanupWiggleMIDs
  (if (TouchPlatform|IsMobileGame)
    (Components|Timeline|Stop (Variables|Tiered_RealityFruit_Athena_Parent|GetMobileOnInteractTL))
    (for _array_element (Variables|Default|GetMIDs)
      (Rendering|Material|SetScalarParameterValue _array_element "WiggleOnInteractAmount" 0.0 :!class /Script/Engine.MaterialInstanceDynamic))))

(fn CreateMobileMIDs
  (if (TouchPlatform|IsMobileGame)
    (for (_ _array_index) (Rendering|Material|GetMaterials (Variables|Debug|GetStaticMeshComponent))
      (bind _returnvalue (CallFunction|GetorCreateMIDOnDefaultMesh _array_index))
      (Utilities|Array|SetArrayElem (Variables|Default|GetMIDs) _array_index _returnvalue true))))

(fn FruitDataRepped)

(fn GetMaxAudibleDistance
  (bind _chest_ambient_sound (Variables|Default|GetChest_Ambient_Sound))
  (bind _sound (Variables|/Script/Engine.AudioComponent|GetSound _chest_ambient_sound))
  (Utilities|IsValid _sound
    (:Is%20Valid
      (return (Variables|/Script/Engine.SoundBase|GetMaxDistance _sound)))
    (:Is%20Not%20Valid
      (return 0.0))))

(fn GetOwningControllerForAnalytics
  (bind (_fruitreppeddata_owningplayercontroller_6_7553902549a37048686212bc13e7d8ca _fruitreppeddata_saplingparent_11_8f3a618f4bd64834d3636281b025ea9d) (Utilities|Struct|BreakFruitReppedData (Variables|GetFruitReppedData_OwningPlayerController_6_7553902549A37048686212BC13E7D8CA)))
  (Utilities|IsValid _fruitreppeddata_saplingparent_11_8f3a618f4bd64834d3636281b025ea9d
    (:Is%20Valid
      (bind _asfort_player_controller_athena (Utilities|Casting|CastToFortPlayerControllerAthena (Variables|/RealitySeedGameplay/items/RealitySaplings/B_SMA_Athena_RealitySapling_Parent.B_SMA_Athena_RealitySapling_Parent_C|GetOwning%20Player%20Controller%20Repped _fruitreppeddata_saplingparent_11_8f3a618f4bd64834d3636281b025ea9d))
        (:then
          (return _asfort_player_controller_athena))
        (:CastFailed
          _)))
    (:Is%20Not%20Valid
      _))
  (return 0))

(fn OnRep_FruitReppedData
  (CallFunction|FruitDataRepped))

(fn OnSetCustomDepthStencilValue
  (params (PrimComponents :type Array%20of%20Primitive%20Component%20Object%20References) (bUseCustomDepth :type Boolean) (StencilValue :type Integer))
  (for _array_element PrimComponents
    (Rendering|SetCustomDepthStencilValue _array_element (Variables|GetStencilValue))
    (if (Variables|GetbUseCustomDepth)
      (Rendering|SetRenderCustomDepth _array_element true)))
  (return true))

(fn SetLootTierGroupFromNurturing
  (params (LocalSaplingParent :type B%20SMA%20Athena%20Reality%20Sapling%20Parent%20Object%20Reference))
  (bind _numtimesweeded (Variables|/RealitySeedGameplay/items/RealitySaplings/B_SMA_Athena_RealitySapling_Parent.B_SMA_Athena_RealitySapling_Parent_C|GetNumTimesWeeded LocalSaplingParent))
  (Utilities|FlowControl|Sequence
    (:then_0
      (if (>= _numtimesweeded 1)
        (switch int _numtimesweeded
          (:Default
            (Variables|Default|SetMaxTierLevel 4))
          (:0
            (Variables|Default|SetMaxTierLevel 1))
          (:1
            (Variables|Default|SetMaxTierLevel 2))
          (:2
            (Variables|Default|SetMaxTierLevel 3))
          (:3
            (Variables|Default|SetMaxTierLevel 5)))))
    (:then_1
      (CallFunction|SetLootTierGroupfromRarity))))

(fn SetLootTierGroupFromRarity)

(fn SetLightVisibility
  (params (Visible :type Boolean)))

(fn SetVisibleMobileInteractIcon
  (params (Visible :type Boolean))
  (bind (_returnvalue_roll _returnvalue_pitch _returnvalue_yaw) (Math|Rotator|BreakRotator (Transformation|GetActorRotation)))
  (if Visible
    ;[1 1 1] Create
    (bind _mobileinteracticon (Variables|Default|GetMobileInteractIcon)
      (:then
        (Rendering|SetVisibility _mobileinteracticon true false))
      (:else
        (Variables|Default|GetMobileInteractionMaterial
          (:then
            (bind _returnvalue (AddComponent|NODE_AddStaticMeshComponent-1 :RelativeTransform_Location (Variables|Default|GetMobileInteractIconLocation) :RelativeTransform_Rotation_Yaw (* _returnvalue_yaw -1.0) :RelativeTransform_Scale (Variables|Default|GetMobileInteractIconScale)))
            (bind _output_get (Variables|Default|SetMobileInteractIcon _returnvalue))
            (Rendering|Material|SetMaterial _output_get 0 (Variables|Default|GetMobileInteractionMaterial)))
          (:else))))
    (else
      ;[1 1 1] Destroy
      (bind _mobileinteracticon_2 (Variables|Default|GetMobileInteractIcon)
        (:then
          (Rendering|SetVisibility _mobileinteracticon_2 false false)
          (Components|DestroyComponent _mobileinteracticon_2))
        (:else)))))

(fn ShouldDie
  (params (Damage :type Float%20%28single-precision%29) (EventInstigator :type Controller%20Object%20Reference) (DamageCauser :type Actor%20Object%20Reference) (bIsTest :type Boolean) (EffectContext :type Gameplay%20Effect%20Context%20Handle%20Structure%20%28by%20ref%29))
  (bind _asfort_player_controller_outpost (Utilities|Casting|CastToFortPlayerControllerOutpost EventInstigator)
    (:then
      (return (Outpost|IsOutpostOwner _asfort_player_controller_outpost)))
    (:CastFailed
      (return true))))

(fn ConstructionScript
  (Utilities|FlowControl|Sequence
    (:then_0
      (Networking|CanExecuteCosmeticEvents
        (:True
          (if (Utilities|String|NotEqualExactly(String) (Game|GetPlatformName) "Windows")
            _))))
    (:then_1
      ;[1 1 1] remove negative scale from all placed chests so vertex shader is correct
      (Transformation|SetActorScale3D (Math|Vector|MakeVector (Math|Float|Absolute(Float) (.x (Transformation|GetActorScale3D))) (Math|Float|Absolute(Float) (.y (Transformation|GetActorScale3D))) (Math|Float|Absolute(Float) (.z (Transformation|GetActorScale3D))))))))
